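; Migration variant of the Movement Modulator V2 Creative device that corrects placement, because the prop sits differently from the original trap.
; Its construction script calls the parent, then runs AdjustTranslationForMigration.
; On authority, if Migrated is false, that function shifts the actor 256 units along its local Y and yaws it -90 degrees.
; It then sets Migrated and calls Resave, so the flag guards against repeating the correction.
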(fn AdjustTranslationForMigration
  (bind _returnvalue (Transformation|GetActorRotation))
  (bind _self self)
  ;[1 0 0.6587 0.25] Location of the Prop is different than the location of the original trap, so we need to move it. A child will set it to true
  (Networking|SwitchHasAuthority
    (:Authority
      (if (not (Variables|Default|GetMigrated))
        (Transformation|SetActorLocation _self (+ (Transformation|GetActorLocation) (Math|Vector|RotateVector "0,256.000000,0.000000" _returnvalue)) false true)
        (Transformation|SetActorRotation _self (Math|Rotator|CombineRotators "0,-90.000000,0.000000" _returnvalue) false)
        (Variables|Default|SetMigrated true)
        (CallFunction|Resave)))))

(fn ConstructionScript
  (|CallParentFunction)
  (CallFunction|AdjustTranslationforMigration))
