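; Migration variant of the Movement Modulator V2 creative device that realigns the placed device, as its prop sits differently from the original trap.
; Its construction script calls the parent, then runs the adjustment on authority only while the Migrated flag is unset.
; The adjustment raises the actor along Z, composes a corrective rotation with its current one and overrides its scale.
; It then sets Migrated and calls Resave, guarding against repeated offsets.
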(fn AdjustTranslationForMigration
  (bind _self self)
  ;[1 0 0.6587 0.25] Location of the Prop is different than the location of the original trap, so we need to move it. A child will set it to true
  (Networking|SwitchHasAuthority
    (:Authority
      (if (not (Variables|Default|GetMigrated))
        (Transformation|SetActorLocation _self (+ (Transformation|GetActorLocation) "0,0.000000,144.000000") false true)
        (Transformation|SetActorRotation _self (Math|Rotator|CombineRotators "90.000000,0.000000,90.000000" (Transformation|GetActorRotation)) false)
        (Transformation|SetActorScale3D "1.000000,0.750000,1.000000")
        (Variables|Default|SetMigrated true)
        (CallFunction|Resave)))))

(fn ConstructionScript
  (|CallParentFunction)
  (CallFunction|AdjustTranslationforMigration))
