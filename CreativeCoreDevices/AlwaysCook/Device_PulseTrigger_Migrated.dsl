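; Migrated variant of the Pulse Trigger creative device that corrects its placement, since the prop's location differs from that of the original trap.
; The construction script calls the parent, then, if the location-migration flag is unset and it has authority, shifts the actor 256 units along its rotated local Y axis.
; It then sets the flag and calls Resave, making the shift one-time.
; The event graph has no active logic.
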(event EventBeginPlay
  (|CallParentFunction) ; This node is disabled and will not be called.\nDrag off pins to build functionality.
)

(event Collision|EventActorBeginOverlap
  (params (OtherActor :type Actor%20Object%20Reference)))

(event EventTick
  (params (DeltaSeconds :type Float%20%28single-precision%29)))

(fn AdjustTranslationForMigration
  ;[1 0 0.6587 0.25] Location of the Prop is different than the location of the original trap, so we need to move it. A child will set it to true
  (if (not (Variables|LegacyMigration|GetLocationMigrationDone))
    (Networking|SwitchHasAuthority
      (:Authority
        (Transformation|SetActorLocation (+ (Transformation|GetActorLocation) (Math|Vector|RotateVector "0,256.000000, 0" (Transformation|GetActorRotation))))
        (Variables|LegacyMigration|SetLocationMigrationDone true)
        (CallFunction|Resave)))))

(fn ConstructionScript
  (|CallParentFunction)
  (CallFunction|AdjustTranslationforMigration))
