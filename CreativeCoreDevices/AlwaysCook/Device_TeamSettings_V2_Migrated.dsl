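; Migrated V2 Team Settings device Blueprint that corrects its placement, since the prop's location differs from that of the original trap it replaces.
; The construction script calls its parent, then runs an authority-only translation adjustment.
; If the legacy Migrated flag is unset, it moves the actor by a fixed offset rotated to its orientation, sets the flag and calls Resave so the shift applies once.
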(event EventTick
  (params (DeltaSeconds :type Float%20%28single-precision%29)))

(fn AdjustTranslationforMigration
  ;[1 0 0.6587 0.25] Location of the Prop is different than the location of the original trap, so we need to move it. A child will set it to true
  (Networking|SwitchHasAuthority
    (:Authority
      (if (not (Variables|Legacy|GetMigrated))
        (Transformation|SetActorLocation self (+ (Transformation|GetActorLocation) (Math|Vector|RotateVector "0,256.000000,192.000000" (Transformation|GetActorRotation))) false true)
        (Variables|Legacy|SetMigrated true)
        (CallFunction|Resave)))))

(fn ConstructionScript
  (|CallParentFunction)
  (CallFunction|AdjustTranslationForMigration))
